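# MF_HelloHoloPBR is a material function in the HelloHolo_PBR wheel assets under VehicleCosmetics.
# Its graph holds a single output node with nothing connected to it.
# It has no inputs, parameters or shading logic, so it reads as an empty stub.
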MaterialFunction /VehicleCosmetics/Wheels/HelloHolo_PBR/Materials/MF_HelloHoloPBR
graph {nodes [FunctionOutput()]}
