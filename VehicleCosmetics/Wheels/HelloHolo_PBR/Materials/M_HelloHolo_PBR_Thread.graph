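# Masked, two-sided unlit material for the thread of the HelloHolo PBR holographic wheel's tire, usable on skeletal meshes.
# Builds scrolling screen-space scanlines under Fresnel, animated UV stripes and a panning gradient, shared through named reroutes.
# Spinner emissive branches tint tire-mask channels with MainColor.
# The Emissive branch adds normal-mapped Fresnel edges and lines from a scrolling geometric texture, reusing the normal map's green channel as detail.
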Material /VehicleCosmetics/Wheels/HelloHolo_PBR/Materials/M_HelloHolo_PBR_Thread
graph {
 properties {
  bAutomaticallySetUsageInEditor=false
  bUsedWithMeshDeformer=true
  bUsedWithSkeletalMesh=true
  blendMode=BLEND_Masked
  shadingModel=MSM_Unlit
  twoSided=true
  usedShadingModels=Unlit
 }
 nodes [
  n1 = TextureSample() {samplerType=SAMPLERTYPE_LinearColor, texture=/VehicleCosmetics/Wheels/HelloHolo_PBR/Textures/T_HelloholoPbr_Tire_M}
  n2 = ScalarParameter() {defaultValue=1, group=TireDetails, parameterName=ProjectorAnimation}
  n3 = NamedRerouteUsage()
  n4 = Multiply(A=@n1.B, B=ScalarParameter() {defaultValue=0.03, group=TireDetails, parameterName=ProjectorPower})
  n5 = Add(A=Multiply(A=Multiply(A=@n2, B=@n3), B=@n4), B=@n4)
  Add(A=Multiply(A=ScalarParameter() {defaultValue=1, parameterName=RingPower}, B=Multiply(A=@n1.B, B=@n1.B)), B=@n5)
  Multiply(A=@n1.B)
  Multiply(A=Multiply(A=@n5, B=ScalarParameter() {defaultValue=5, parameterName=FresnelMultiplierLine}), B=NamedRerouteUsage())
  NamedRerouteDeclaration(None=VertexColor().R) {name=VC_R, nodeColor=<rgba 1 0.046 0 1>}
  n6 = ComponentMask(None=TextureCoordinate() {vTiling=35}) {g=true}
  n7 = Multiply(A=Constant() {r=1.2}, B=Time() {bIgnorePause=true, period=3})
  n8 = Frac(None=ComponentMask(None=Panner(Coordinate=@n6, Time=@n7) {speedX=1.1, speedY=-6}) {r=true})
  n9 = OneMinus(None=@n8)
  NamedRerouteDeclaration(None=ConstantBiasScale(None=Multiply(A=@n8, B=@n9)) {bias=0, scale=2}) {name=StripesMov, nodeColor=<rgba 1 0.3633 0.6824 1>}
  n10 = Fresnel() {exponent=1}
  n11 = ComponentMask(None=ScreenPosition().PixelPosition) {g=true}
  n12 = ScalarParameter() {defaultValue=1.8, group=TireDetails, parameterName=StripesSize}
  n13 = Frac(None=Subtract(A=Multiply(A=@n11, B=@n12), B=Time()))
  n14 = Clamp(None=Multiply(A=@n13, B=OneMinus(None=@n13))) {maxDefault=3}
  NamedRerouteDeclaration(None=ConstantBiasScale(None=Multiply(A=@n10, B=@n14)) {bias=0, scale=1.3}) {name=GlobalGradient, nodeColor=<rgba 0.5703 1 0 1>}
  n15 = TextureSample(UVs=Panner(Coordinate=TextureCoordinate() {uTiling=0.4}) {speedX=-0.35}) {texture=/VehicleCosmetics/Wheels/_SharedTextures/FX/Gradient_D}
  NamedRerouteDeclaration(None=LinearInterpolate(A=ConstantBiasScale(None=@n15.R), B=Constant() {r=1}, Alpha=NamedRerouteUsage())) {name=Scroll, nodeColor=<rgba 0.1232 0.25 1 1>}
  n16 = NamedRerouteUsage()
  n17 = NamedRerouteUsage()
  n18 = TextureSample() {samplerType=SAMPLERTYPE_LinearColor, texture=&T_HelloholoPbr_Tire_M}
  n19 = Multiply(A=@n18.R, B=ScalarParameter() {defaultValue=4, group=TireDetails, parameterName=FresnelMultiplier})
  n20 = Multiply(A=@n18.B, B=ScalarParameter() {defaultValue=0.03, group=TireDetails, parameterName=StripesSidePower})
  n21 = Multiply(A=Multiply(A=ScalarParameter() {defaultValue=1, group=TireDetails, parameterName=StripesSideProjection}, B=NamedRerouteUsage()), B=@n20)
  n22 = Multiply(
   A=ScalarParameter() {defaultValue=1.5, group=TireDetails, parameterName=EmissiveMultiplier}
   B=Add(A=Multiply(A=@n17, B=@n19), B=Add(A=@n21, B=@n20))
  )
  NamedRerouteDeclaration(None=Multiply(A=VectorParameter() {group=TireDetails, parameterName=MainColor, defaultValue=<rgba 0.5156 0.2372 0 1>}.RGB, B=Multiply(A=@n16, B=@n22))) {name=Emissive_Spinner, nodeColor=<rgba 1 0 0.4922 1>}
  n23 = ComponentMask(None=ScreenPosition().PixelPosition) {g=true}
  n24 = ScalarParameter() {defaultValue=1.8, group=TireDetails, parameterName=StripesSize}
  n25 = Frac(None=Subtract(A=Multiply(A=@n23, B=@n24), B=Time()))
  n26 = Multiply(A=@n25, B=OneMinus(None=@n25))
  NamedRerouteDeclaration(None=Multiply(A=Fresnel() {exponent=1}, B=Clamp(None=@n26))) {name=GlobalGradient_1, nodeColor=<rgba 0.5703 1 0 1>}
  n27 = NamedRerouteUsage()
  n28 = VectorParameter() {group=FresnelTire, parameterName=FresnelParametr, defaultValue=<rgba 1 0.3953 0.0114 1>}
  n29 = MaterialFunctionCall!(
   Input Normal=TextureSample() {samplerType=SAMPLERTYPE_Normal, texture=/VehicleCosmetics/Wheels/HelloHolo_PBR/Textures/T_HelloholoPbr_Tire_N}.RGB
   Input Sharpness=ScalarParameter() {defaultValue=1.5, group=FresnelTire, parameterName=FresnelSharpness}
   Input Two Sided
  ) {materialFunction=/VehicleCosmetics/Wheels/CarCar_Holographic/Materials/MF_Fresnel_Advanced}
  n30 = LinearInterpolate(A=@n27, B=LinearInterpolate(A=@n27, B=Multiply(A=@n28.RGB, B=@n29.Edge)), Alpha=NamedRerouteUsage())
  n31 = VectorParameter() {group=FresnelTire, parameterName=FresnelParametr, defaultValue=<rgba 1 0.3953 0.0114 1>}
  n32 = Multiply(
   A=Multiply(
    A=ScalarParameter() {defaultValue=0.2, group=Lines, parameterName=LinesPower}
    B=TextureSample() {samplerType=SAMPLERTYPE_Normal, texture=&T_HelloholoPbr_Tire_N}.G
   )
   B=@n31.RGB
  )
  n33 = Frac(None=Panner(Coordinate=TextureCoordinate() {uTiling=2}) {speedX=0.4})
  n34 = Multiply(
   A=Multiply(A=@n31.RGB, B=ScalarParameter() {defaultValue=10, group=Lines, parameterName=EmissiveLines})
   B=Multiply(
    A=TextureSample(UVs=@n33) {samplerSource=SSM_Wrap_WorldGroupSettings, texture=/VehicleCosmetics/Wheels/_SharedTextures/FX/Skin_GeometricBM_RGB}.R
    B=TextureSample() {samplerType=SAMPLERTYPE_LinearColor, texture=&T_HelloholoPbr_Tire_M}.B
   )
  )
  n35 = Multiply(
   A=ScalarParameter() {defaultValue=1.8, group=Lines, parameterName=LinesContrast}
   B=ComponentMask(None=TextureSample() {samplerType=SAMPLERTYPE_Normal, texture=&T_HelloholoPbr_Tire_N}.RGB) {g=true}
  )
  NamedRerouteDeclaration(None=Add(A=Add(A=@n30, B=@n32), B=Multiply(A=@n34, B=@n35))) {name=Emissive, nodeColor=<rgba 0.5637 0.2969 1 1>}
  n36 = NamedRerouteUsage()
  n37 = Multiply(A=NamedRerouteUsage(), B=Multiply(A=@n1.R, B=ScalarParameter() {defaultValue=4, group=TireDetails, parameterName=FresnelMultiplier}))
  n38 = Multiply(A=ScalarParameter() {defaultValue=1.5, group=TireDetails, parameterName=EmissiveMultiplier}, B=Add(A=@n37, B=@n5))
  NamedRerouteDeclaration(None=Multiply(A=VectorParameter() {group=TireDetails, parameterName=MainColor, defaultValue=<rgba 0.5156 0.2372 0 1>}.RGB, B=Multiply(A=@n36, B=@n38))) {name=Emissive_Spinner_2, nodeColor=<rgba 1 0 0.4922 1>}
  Reroute(None=Add(A=@n37))
 ]
}
