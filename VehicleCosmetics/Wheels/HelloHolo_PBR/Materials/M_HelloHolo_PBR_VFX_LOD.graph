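# Masked, two-sided unlit material for the holographic VFX on the HelloHolo PBR wheel cosmetic, named as its LOD version.
# Blends channel-packed tread and Holo3 textures into outline, ring and projector terms, tinted by MainColor for emissive.
# Builds animated modulators from screen-pixel stripes with normal-mapped Fresnel, panned UV stripes and a scrolling gradient.
# Thresholded sine waves create intermittent opacity flicker, and a ShadowReplace branch multiplies with spinner texture alpha.
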Material /VehicleCosmetics/Wheels/HelloHolo_PBR/Materials/M_HelloHolo_PBR_VFX_LOD
graph {
 properties {
  bAutomaticallySetUsageInEditor=false
  bUsedWithMeshDeformer=true
  bUsedWithSkeletalMesh=true
  blendMode=BLEND_Masked
  shadingModel=MSM_Unlit
  twoSided=true
  usedShadingModels=Unlit
 }
 nodes [
  Multiply(
   A=ShadowReplace(Default=NamedRerouteUsage(), Shadow=ScalarParameter())
   B=TextureSample() {texture=/VehicleCosmetics/Wheels/HelloHolo_PBR/Textures/T_HelloHoloPbr_Spinner_BC}.A
  )
  n1 = ComponentMask(None=TextureCoordinate() {uTiling=10, vTiling=15}) {g=true}
  n2 = Multiply(A=Constant() {r=1.2}, B=Time() {bIgnorePause=true, period=3})
  n3 = Frac(None=ComponentMask(None=Panner(Coordinate=@n1, Time=@n2) {speedX=-0.4, speedY=-1}) {r=true})
  n4 = OneMinus(None=@n3)
  n5 = ComponentMask(None=TextureCoordinate() {vTiling=25}) {g=true}
  n6 = Multiply(A=Constant() {r=1.2}, B=Time() {bIgnorePause=true, period=3})
  n7 = Frac(None=ComponentMask(None=Panner(Coordinate=@n5, Time=@n6) {speedX=0.3, speedY=-5}) {r=true})
  n8 = OneMinus(None=@n7)
  n9 = ComponentMask(None=TextureCoordinate() {vTiling=25}) {g=true}
  n10 = Multiply(A=Constant() {r=1.2}, B=Time() {bIgnorePause=true, period=3})
  n11 = Frac(None=ComponentMask(None=Panner(Coordinate=@n9, Time=@n10) {speedX=1.1, speedY=-6}) {r=true})
  n12 = ConstantBiasScale(None=Multiply(A=@n11, B=OneMinus(None=@n11))) {bias=0, scale=2}
  Multiply(A=Multiply(A=Multiply(A=@n3, B=@n4), B=Multiply(A=@n7, B=@n8)), B=@n12)
  NamedRerouteDeclaration(None=VertexColor().R) {name=VC_R, nodeColor=<rgba 1 0.046 0 1>}
  NamedRerouteDeclaration(None=@n12) {name=StripesMov, nodeColor=<rgba 1 0.3633 0.6824 1>}
  n13 = TextureSample() {samplerType=SAMPLERTYPE_Normal, texture=/VehicleCosmetics/Wheels/HelloHolo_PBR/Textures/T_HelloHoloPbr_Spinner_N}
  n14 = ComponentMask(None=ScreenPosition().PixelPosition) {g=true}
  n15 = ScalarParameter() {defaultValue=1.1, parameterName=StripesSize}
  n16 = Frac(None=Subtract(A=Multiply(A=@n14, B=@n15), B=Time()))
  n17 = Multiply(A=@n16, B=OneMinus(None=@n16))
  NamedRerouteDeclaration(None=Multiply(A=Fresnel(Normal=@n13.RGB) {exponent=1}, B=Clamp(None=@n17))) {name=GlobalGradient, nodeColor=<rgba 0.5703 1 0 1>}
  n18 = TextureSample(UVs=Panner(Coordinate=TextureCoordinate() {uTiling=0.4}) {speedX=-0.35}) {texture=/VehicleCosmetics/Wheels/_SharedTextures/FX/Gradient_D}
  NamedRerouteDeclaration(None=LinearInterpolate(A=ConstantBiasScale(None=@n18.R), B=Constant() {r=1}, Alpha=NamedRerouteUsage())) {name=Scroll, nodeColor=<rgba 0.1232 0.25 1 1>}
  n19 = NamedRerouteUsage()
  n20 = TextureSample() {samplerType=SAMPLERTYPE_LinearColor, texture=/VehicleCosmetics/Wheels/HelloHolo_PBR/Textures/T_HelloHolo_Tread}
  n21 = TextureSample() {samplerType=SAMPLERTYPE_LinearColor, texture=/VehicleCosmetics/Wheels/HelloHolo_PBR/Textures/T_Holo3}
  n22 = NamedRerouteUsage()
  n23 = LinearInterpolate(A=@n20.G, B=@n21.G, Alpha=@n22) {desc=Green}
  n24 = LinearInterpolate(A=@n20.A, B=@n21.A, Alpha=@n22) {desc=Alpha}
  n25 = OneMinus(None=ConstantBiasScale(None=Add(A=@n23, B=@n24)) {bias=0, scale=2})
  n26 = ScalarParameter() {defaultValue=5, parameterName=FresnelMultiplier}
  n27 = LinearInterpolate(A=@n20.R, B=@n21.R, Alpha=@n22) {desc=Red}
  n28 = Add(
   A=Multiply(A=ScalarParameter() {defaultValue=0.1, parameterName=Outline_power}, B=Multiply(A=@n27, B=@n25))
   B=Multiply(A=NamedRerouteUsage(), B=Multiply(A=@n25, B=@n26))
  )
  n29 = ScalarParameter() {defaultValue=1, parameterName=RingPower}
  n30 = Multiply(A=@n27, B=@n23)
  n31 = Multiply(A=Multiply(A=@n27, B=@n24), B=ScalarParameter() {defaultValue=1, parameterName=ProjectorPower})
  n32 = Multiply(A=Multiply(A=ScalarParameter() {defaultValue=3, parameterName=ProjectorAnimation}, B=NamedRerouteUsage()), B=@n31)
  n33 = Add(A=@n28, B=Add(A=Multiply(A=@n29, B=@n30), B=Add(A=@n32, B=@n31)))
  n34 = Multiply(A=ScalarParameter() {defaultValue=1, parameterName=EmissiveMultiplier}, B=Add(A=@n28, B=@n33))
  NamedRerouteDeclaration(None=Multiply(A=VectorParameter() {parameterName=MainColor, defaultValue=<rgba 1 0.46 0 1>}.RGB, B=Multiply(A=@n19, B=@n34))) {name=Emissive_Spinner, nodeColor=<rgba 1 0 0.4922 1>}
  n35 = Multiply(A=Time(), B=ScalarParameter() {defaultValue=10, parameterName=Flicker_Speed})
  n36 = Constant()
  n37 = ScalarParameter() {defaultValue=0.9, parameterName=Flicker_Contrast}
  n38 = Add(A=If(A=Sine(None=@n35), B=@n36, "A > B"=@n36, "A == B"=@n36, "A < B"=Constant() {r=1}), B=@n37)
  n39 = Multiply(A=@n33, B=Clamp(None=Multiply(A=@n38, B=@n37)))
  n40 = Multiply(A=Time(), B=ScalarParameter() {defaultValue=0.3, parameterName=Flicker_Interval})
  n41 = Constant() {r=1}
  n42 = ScalarParameter() {defaultValue=0.05, parameterName=FlickerLength}
  n43 = Constant()
  n44 = If(A=Add(A=Sine(None=@n40), B=Subtract(A=@n41, B=@n42)), B=@n43, "A > B"=@n43, "A == B"=@n43, "A < B"=Constant() {r=1})
  NamedRerouteDeclaration(None=Add(A=ScalarParameter() {defaultValue=0.15, parameterName=OpacityOffset}, B=LinearInterpolate(A=@n33, B=@n39, Alpha=@n44))) {name=Opacity, nodeColor=<rgba 0.3698 0.3698 0.3698 1>}
  NamedRerouteUsage()
  NamedRerouteUsage()
  Reroute()
 ]
}
