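# Masked, Default Lit material for the HelloHolo PBR vehicle wheel cosmetic, built on material attributes and flagged for skeletal mesh and mesh deformer use.
# The shared simple alpha wheel parent function supplies the surface, which passes through base vehicle damage and then the vehicle de-rez transition.
# A TIRE reroute taken from the parent's red vertex color and a time-scaled UV panner have no consumers in the graph.
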Material /VehicleCosmetics/Wheels/HelloHolo_PBR/Materials/M_Wheel_HelloHolo_PBR
graph {
 properties {
  bAutomaticallySetUsageInEditor=false
  bCastRayTracedShadows=false
  bUseMaterialAttributes=true
  bUsedWithMeshDeformer=true
  bUsedWithSkeletalMesh=true
  blendMode=BLEND_Masked
  pixelDepthOffsetMode=PDOM_Legacy
  usedShadingModels=Default Lit
 }
 nodes [
  n1 = MaterialFunctionCall() {materialFunction=/VehicleCosmetics/Wheels/Mage/Materials/MF_WheelParent_Simple_Alpha}
  MaterialFunctionCall(In=MaterialFunctionCall(Material=@n1.MaterialAttributes) {materialFunction=&MF_BaseVehicle_Damage}.Result) {materialFunction=&MF_VehicleDeRez_Transition}
  NamedRerouteDeclaration(None=@n1.VertexColorR) {name=TIRE, nodeColor=<rgba 1 0.0819 0 1>}
  Panner(Coordinate=ComponentMask(None=TextureCoordinate()) {g=true}, Time=Multiply(A=ScalarParameter() {defaultValue=1.2}, B=Time())) {speedX=1, speedY=2}
 ]
}
